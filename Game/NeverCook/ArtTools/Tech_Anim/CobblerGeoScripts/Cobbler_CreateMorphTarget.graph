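# Editor utility Blueprint of experimental Geometry Script tools that build shoe morph targets on selected skeletal mesh actors.
# The main function transfers weights from the second mesh to the first, then collapses second-mesh vertices inside the first's bounds onto foot_l as morph shoes_mid.
# A variant selects vertices by polygroup.
# Test functions rebind bones with materials and convert render-data LODs, saving to hard-coded asset paths and skeletons.
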EditorUtilityBlueprint /Game/NeverCook/ArtTools/Tech_Anim/CobblerGeoScripts/Cobbler_CreateMorphTarget
graphs [
 Cobbler_CreateMorphTarget {nodes [
   n1 = GetSelectionSet!(execute) -> (then, ReturnValue:Array of Actor Object References)
   > n2 = Cast To StaticMeshActor!(execute, Object="Get (a copy)"(Array:Array of Actor Object References=@n1) -> (Output:Actor Object Reference)) -> (then, CastFailed, AsStatic Mesh Actor:Static Mesh Actor Object Reference) {is=title K2Node_DynamicCast}
   > n3 = Construct Object from Class(Class=/Script/GeometryFramework.DynamicMesh) -> (ReturnValue:Dynamic Mesh Object Reference)
   > n4 = Copy Mesh From Static Mesh with Section Materials!(
    execute
    FromStaticMeshAsset:Static Mesh Object Reference=Get StaticMesh(self=Get StaticMeshComponent(self=@n2))
    ToDynamicMesh:Dynamic Mesh Object Reference=@n3
    AssetOptions:Geometry Script Copy Mesh From Asset Options Structure
    RequestedLOD:Geometry Script Mesh Read LOD Structure
   ) -> (Failure:Exec, Success:Exec, ReturnValue:Dynamic Mesh Object Reference)
   TransformMesh!(
    execute=@n4.Success
    TargetMesh:Dynamic Mesh Object Reference=@n4
    Transform:Transform=Get Actor Transform(self=@n2)
    bFixOrientationForNegativeScale:bool=true
   ) -> (then, ReturnValue:Dynamic Mesh Object Reference)
   n5 = Make Index List!(IndexType:EGeometryScriptIndexType Enum=Triangle) -> (GeometryScriptIndexList:Index List Structure) {is=title K2Node_MakeStruct}
   n6 = Get Triangles In PolyGroup!(
    execute
    TargetMesh:Dynamic Mesh Object Reference
    GroupLayer:Geometry Script Group Layer Structure=Make Geometry Script Group Layer!(bDefaultLayer:bool=True, ExtendedLayerIndex:int=0) -> (GeometryScriptGroupLayer:Geometry Script Group Layer Structure) {is=title K2Node_MakeStruct}
    PolygroupID:int=0
    TriangleIDsOut:"Index List Structure (by ref)"=@n5
   ) -> (then, ReturnValue:Dynamic Mesh Object Reference) {is=title K2Node_CallFunction}
   > ConvertIndexListToMeshSelection!(
    execute
    TargetMesh:Dynamic Mesh Object Reference=@n6
    IndexList:Index List Structure=@n5
    SelectionType:EGeometryScriptMeshSelectionType Enum=Triangles
   ) -> (then, Selection:Mesh Selection Structure, ReturnValue:Dynamic Mesh Object Reference)
   Cobbler_CreateMorphTarget() {is=title K2Node_FunctionEntry}
   > n7 = Sequence!(execute) -> (then_0, then_1, then_2:Exec)
   > n8 = GetSelectionSet()
   > n9 = Cast To SkeletalMeshActor!(execute, Object="Get (a copy)"(Array=@n8)) -> (then, CastFailed, AsSkeletal Mesh Actor:Skeletal Mesh Actor Object Reference) {is=title K2Node_DynamicCast}
   > n10 = Construct Object from Class(Class=&DynamicMesh)
   n11 = Get SkeletalMeshAsset!(self:Skeletal Mesh Component Object Reference=Get SkeletalMeshComponent!(self:Skeletal Mesh Actor Object Reference=@n9) -> (SkeletalMeshComponent:Skeletal Mesh Component Object Reference)) -> (SkeletalMeshAsset:Skeletal Mesh Object Reference)
   > n12 = CopyMeshFromSkeletalMesh!(
    execute
    FromSkeletalMeshAsset:Skeletal Mesh Object Reference=@n11
    ToDynamicMesh:Dynamic Mesh Object Reference=@n10
    AssetOptions:Geometry Script Copy Mesh From Asset Options Structure
    RequestedLOD:Geometry Script Mesh Read LOD Structure
   ) -> (Failure:Exec, Success:Exec, ReturnValue:Dynamic Mesh Object Reference)
   n13 = TransformMesh(execute=@n12.Success, TargetMesh=@n12, Transform=Get Actor Transform(self=@n9), bFixOrientationForNegativeScale=true)
   n14 = GetSelectionSet(execute=@n7.then_1)
   > n15 = Cast To SkeletalMeshActor(Object="Get (a copy)"(Array=@n14, Dimension 1=1))
   > n16 = Construct Object from Class(Class=&DynamicMesh)
   n17 = Get SkeletalMeshAsset(self=Get SkeletalMeshComponent(self=@n15))
   > n18 = CopyMeshFromSkeletalMesh(FromSkeletalMeshAsset=@n17, ToDynamicMesh=@n16)
   n19 = TransformMesh(execute=@n18.Success, TargetMesh=@n18, Transform=Get Actor Transform(self=@n15), bFixOrientationForNegativeScale=true)
   > n20 = TransferBoneWeightsFromMesh!(
    execute
    SourceMesh:Dynamic Mesh Object Reference=@n19
    TargetMesh:Dynamic Mesh Object Reference=@n13
    Options:Geometry Script Transfer Bone Weights Options Structure=Make Geometry Script Transfer Bone Weights Options!(
     TransferMethod:ETransferBoneWeightsMethod Enum=ClosestPointOnSurface
     OutputTargetMeshBones:EOutputTargetMeshBones Enum=TargetBones
     SourceProfile:Bone Weights Profile Structure=(ProfileName="Default")
     TargetProfile:Bone Weights Profile Structure=(ProfileName="Default")
     RadiusPercentage:double=-1
     NormalThreshold:double=-1
     LayeredMeshSupport:bool=True
     NumSmoothingIterations:int=0
     SmoothingStrength:float=0
     InpaintMask:Name=None
    ) -> (GeometryScriptTransferBoneWeightsOptions:Geometry Script Transfer Bone Weights Options Structure) {is=title K2Node_MakeStruct}
    Selection:Mesh Selection Structure=()
   ) -> (then, ReturnValue:Dynamic Mesh Object Reference)
   > n21 = CreateNewSkeletalMeshAssetFromMesh!(
    execute
    FromDynamicMesh:Dynamic Mesh Object Reference=@n20
    InSkeleton:Skeleton Object Reference=/BRCosmetics/Characters/Player/Female/Medium/Bodies/F_MED_AgentSherbert/Meshes/F_MED_AgentSherbert_Skeleton
    AssetPathAndName:String=/Game/__ARinat/Agent/Morph/Shoes
    Options:Geometry Script Create New Skeletal Mesh Asset Options Structure
   ) -> (Failure:Exec, Success:Exec, ReturnValue:Skeletal Mesh Object Reference)
   n22 = CopyMeshToSkeletalMesh!(
    execute=@n21.Success
    FromDynamicMesh:Dynamic Mesh Object Reference=@n20
    ToSkeletalMeshAsset:Skeletal Mesh Object Reference=@n11
    Options:Geometry Script Copy Mesh To Asset Options Structure
    TargetLOD:Geometry Script Mesh Write LOD Structure
   ) -> (Failure:Exec, Success:Exec, ReturnValue:Dynamic Mesh Object Reference)
   n23 = GetBoneInfo!(execute=@n22.Success, TargetMesh:Dynamic Mesh Object Reference=@n19, BoneName:Name=foot_l) -> (then, bIsValidBoneName:bool, BoneInfo:Bone Info Structure, ReturnValue:Dynamic Mesh Object Reference)
   > n24 = SelectMeshElementsInBox!(
    execute
    TargetMesh:Dynamic Mesh Object Reference=@n23.ReturnValue
    Box:Box Structure=GetMeshBoundingBox!(TargetMesh:Dynamic Mesh Object Reference=@n13) -> (ReturnValue:Box Structure)
    SelectionType:EGeometryScriptMeshSelectionType Enum=Vertices
    bInvert:bool=false
    MinNumTrianglePoints:int=3
   ) -> (then, Selection:Mesh Selection Structure, ReturnValue:Dynamic Mesh Object Reference)
   > n25 = ConvertMeshSelectionToIndexArray!(execute, TargetMesh:Dynamic Mesh Object Reference=@n24.ReturnValue, Selection:Mesh Selection Structure=@n24) -> (then, IndexArray:Array of Integers, SelectionType:EGeometryScriptMeshSelectionType Enum=Vertices, ReturnValue:Dynamic Mesh Object Reference)
   > n26 = For Each Loop(Array:Array of Integers=@n25) -> (Array Element:int)
   n27 = Break Bone Info!(GeometryScriptBoneInfo:"Bone Info Structure (by ref)"=@n23.BoneInfo) -> (Index:int, Name:Name, ParentIndex:int, LocalTransform:Transform, WorldTransform:Transform, Color:Linear Color Structure) {is=title K2Node_BreakStruct}
   > n28 = SetVertexPosition!(
    execute
    TargetMesh:Dynamic Mesh Object Reference=@n25.ReturnValue
    VertexID:int=@n26
    NewPosition:Vector=BreakTransform(InTransform=@n27.WorldTransform)
    bDeferChangeNotifications:bool=false
   ) -> (then, bIsValidVertex:bool, ReturnValue:Dynamic Mesh Object Reference)
   > n29 = SetVertexBoneWeights!(
    execute
    TargetMesh:Dynamic Mesh Object Reference=@n28.ReturnValue
    VertexID:int=@n26
    BoneWeights:Array of Bone Weights Structures=Make Array("[0]":Bone Weights Structure=Make Bone Weights!(BoneIndex:int=@n27, Weight:float=1) -> (GeometryScriptBoneWeight:Bone Weights Structure) {is=title K2Node_MakeStruct}) -> (Array:Array of Bone Weights Structures)
    Profile:Bone Weights Profile Structure=(ProfileName="Default")
   ) -> (then, bIsValidVertexID:bool, ReturnValue:Dynamic Mesh Object Reference)
   n30 = CopyMorphTargetToSkeletalMesh!(
    execute=@n26.Completed
    FromMorphTarget:Dynamic Mesh Object Reference=@n29.ReturnValue
    ToSkeletalMeshAsset:Skeletal Mesh Object Reference=@n17
    MorphTargetName:Name=shoes_mid
    Options:Geometry Script Copy Morph Target To Asset Options Structure
    TargetLOD:Geometry Script Mesh Write LOD Structure
   ) -> (Failure:Exec, Success:Exec, ReturnValue:Dynamic Mesh Object Reference)
   CreateNewSkeletalMeshAssetFromMesh(
    execute=@n30.Success
    FromDynamicMesh=@n29.ReturnValue
    InSkeleton=&F_MED_AgentSherbert_Skeleton
    AssetPathAndName=/Game/__ARinat/Agent/Morph/Morph2
   )
  ]}
 Cobbler_PolygroupSelection {nodes [
   Cobbler_PolygroupSelection() {is=title K2Node_FunctionEntry}
   > n1 = GetSelectionSet()
   > n2 = Cast To SkeletalMeshActor(Object="Get (a copy)"(Array=@n1))
   > n3 = Construct Object from Class(Class=&DynamicMesh)
   n4 = Get SkeletalMeshAsset(self=Get SkeletalMeshComponent(self=@n2))
   > n5 = CopyMeshFromSkeletalMesh(FromSkeletalMeshAsset=@n4, ToDynamicMesh=@n3)
   n6 = Make Index List(IndexType=Triangle)
   n7 = Get Triangles In PolyGroup(execute=@n5.Success, TargetMesh=@n5, PolygroupID=100, TriangleIDsOut=@n6)
   > n8 = ConvertIndexListToMeshSelection(TargetMesh=@n7, IndexList=@n6, SelectionType=Vertices)
   > n9 = ConvertMeshSelectionToIndexArray(TargetMesh=@n8.ReturnValue, Selection=@n8) -> (SelectionType=Vertices)
   > n10 = GetBoneInfo(TargetMesh=@n9.ReturnValue, BoneName=foot_l)
   > n11 = For Each Loop(Array=@n9)
   n12 = Break Bone Info(GeometryScriptBoneInfo=@n10.BoneInfo)
   > n13 = SetVertexPosition(TargetMesh=@n10.ReturnValue, VertexID=@n11, NewPosition=BreakTransform(InTransform=@n12.WorldTransform), bDeferChangeNotifications=false)
   > n14 = SetVertexBoneWeights(
    TargetMesh=@n13.ReturnValue
    VertexID=@n11
    BoneWeights=Make Array("[0]"=Make Bone Weights(BoneIndex=@n12, Weight=1))
    Profile=(ProfileName="Default")
   )
   CopyMorphTargetToSkeletalMesh(execute=@n11.Completed, FromMorphTarget=@n14.ReturnValue, ToSkeletalMeshAsset=@n4, MorphTargetName=shoes_mid_)
  ]}
 Cobbler_TestRebind {nodes [
   Cobbler_TestRebind() {is=title K2Node_FunctionEntry}
   > n1 = Sequence()
   > n2 = GetSelectionSet()
   > n3 = Cast To SkeletalMeshActor(Object="Get (a copy)"(Array=@n2))
   > n4 = Construct Object from Class(Class=&DynamicMesh)
   > n5 = CopyMeshFromSkeletalMesh(FromSkeletalMeshAsset=Get SkeletalMeshAsset(self=Get SkeletalMeshComponent(self=@n3)), ToDynamicMesh=@n4)
   n6 = GetSelectionSet(execute=@n1.then_1)
   > n7 = Cast To SkeletalMeshActor(Object="Get (a copy)"(Array=@n6, Dimension 1=1))
   > n8 = Construct Object from Class(Class=&DynamicMesh)
   n9 = Get SkeletalMeshComponent(self=@n7)
   > n10 = CopyMeshFromComponent!(
    execute
    Component:Scene Component Object Reference=@n9
    ToDynamicMesh:Dynamic Mesh Object Reference=@n8
    Options:Geometry Script Copy Mesh From Component Options Structure
    bTransformToWorld:bool=false
   ) -> (Failure:Exec, Success:Exec, LocalToWorld:Transform, ReturnValue:Dynamic Mesh Object Reference)
   n11 = For Each Loop(Exec=@n10.Success, Array:Array of Names=GetMaterialSlotNames!(self:Primitive Component Object Reference=@n9) -> (ReturnValue:Array of Names)) -> (Array Element:Name)
   n12 = CopyBonesFromMesh!(
    execute=@n11.Completed
    SourceMesh:Dynamic Mesh Object Reference=@n5
    TargetMesh:Dynamic Mesh Object Reference=@n10.ReturnValue
    Options:Geometry Script Copy Bones From Mesh Options Structure=Make Geometry Script Copy Bones from Mesh Options!(ReindexWeights:bool=true, BonesToCopyFromSource:EBonesToCopyFromSource Enum=AllBones) -> (GeometryScriptCopyBonesFromMeshOptions:Geometry Script Copy Bones From Mesh Options Structure) {is=title K2Node_MakeStruct}
   ) -> (then, ReturnValue:Dynamic Mesh Object Reference)
   > CreateNewSkeletalMeshAssetFromMesh(
    FromDynamicMesh=@n12
    InSkeleton=&F_MED_AgentSherbert_Skeleton
    AssetPathAndName=/Game/__ARinat/Agent/Morph/RebindShoes
    Options=Make Geometry Script Create New Skeletal Mesh Asset Options!(
     bEnableRecomputeNormals:bool=False
     bEnableRecomputeTangents:bool=False
     Materials:Map of Names to Material Interface Object References=Get MaterialsMap() -> (MaterialsMap:Map of Names to Material Interface Object References)
     bUseMeshBoneProportions:bool=False
     bApplyNaniteSettings:bool=False
     NaniteSettings:Mesh Nanite Settings Structure=(bEnabled=False,bExplicitTangents=False,bLerpUVs=True,bSeparable=False,bVoxelNDF=True,bVoxelOpacity=False,ShapePreservation=None,PositionPrecision=-2147483648,NormalPrecision=-1,TangentPrecision=-1,BoneWeightPrecision=-1,TargetMinimumResidencyInKB=0,KeepPercentTriangles=1,TrimRelativeError=0,GenerateFallback=PlatformDefault,FallbackTarget=Auto,FallbackPercentTriangles=1,FallbackRelativeError=1,MaxEdgeLengthFactor=0,NumRays=64,VoxelLevel=0,RayBackUp=0,DisplacementUVChannel=0,DisplacementMaps=,NaniteAssemblyData=(Parts=,Nodes=))
     bUseOriginalVertexOrder:bool=False
    ) -> (GeometryScriptCreateNewSkeletalMeshAssetOptions:Geometry Script Create New Skeletal Mesh Asset Options Structure) {is=title K2Node_MakeStruct}
   )
   Add!(
    execute=@n11.LoopBody
    TargetMap:Map of Names to Material Interface Object References=Get MaterialsMap()
    Key:"Name (by ref)"=@n11
    Value:"Material Interface Object Reference (by ref)"=GetMaterialByName!(self:Primitive Component Object Reference=@n9, MaterialSlotName:Name=@n11) -> (ReturnValue:Material Interface Object Reference)
   ) -> (then) {is=title K2Node_CallFunction}
  ]}
 EventGraph {links [], nodes []}
 Test Render Data Conversion {nodes [
   Test Render Data Conversion() {is=title K2Node_FunctionEntry}
   > n1 = GetSelectionSet()
   > n2 = Cast To SkeletalMeshActor(Object="Get (a copy)"(Array=@n1))
   > n3 = Construct Object from Class(Class=&DynamicMesh)
   > n4 = CopyMeshFromSkeletalMesh(
    FromSkeletalMeshAsset=Get SkeletalMeshAsset(self=Get SkeletalMeshComponent(self=@n2))
    ToDynamicMesh=@n3
    RequestedLOD=Make Geometry Script Mesh Read LOD!(LODType:EGeometryScriptLODType Enum=RenderData, LODIndex:int=1) -> (GeometryScriptMeshReadLOD:Geometry Script Mesh Read LOD Structure) {is=title K2Node_MakeStruct}
   )
   CreateNewSkeletalMeshAssetFromMesh(
    execute=@n4.Success
    FromDynamicMesh=@n4
    InSkeleton=/Game/Characters/Player/Male/Medium/Bodies/M_MED_Pirate_Progressive/Meshes/M_MED_Pirate_T3_Body_Skeleton
    AssetPathAndName=/Game/__ARinat/Agent/RenderData
   )
   Get Actor Transform(self=@n2)
  ]}
]
variables []
