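# Editor utility that uses Geometry Script to fit selected shoe and body skeletal meshes on Cobblers characters.
# It rebinds shoes to body bones and collapses body vertices inside named bounds boxes onto the calf-to-foot line, saved as morph target shoes_mid.
# It also writes an alternate skin weight profile blending leg vertices toward the calf.
# A variant function blends in weights transferred from a leg template and exports test meshes.
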EditorUtilityBlueprint /Game/NeverCook/ArtTools/Tech_Anim/CobblerGeoScripts/Cobblers_CreateMorphTarget
graphs [
 Cobblers_CreateAltSkinWeights {nodes [
   n1 = GetAllVertexIDs!(execute, TargetMesh:Dynamic Mesh Object Reference=Get UnModified_Body_Mesh() -> (UnModified_Body_Mesh:Dynamic Mesh Object Reference)) -> (then, VertexIDList:Index List Structure, bHasVertexIDGaps:bool, ReturnValue:Dynamic Mesh Object Reference)
   > ConvertIndexListToArray!(execute, IndexList:Index List Structure=@n1) -> (then, IndexArray:Array of Integers)
   n2 = Cobblers_GetLegVertices!(
    execute
    self
    Calf_Name:String
    Thigh_Name:String="Get (a copy)"(Array=Get ThighName_List() -> (ThighName_List:Array of Strings), Dimension 1=1)
    Mesh:Dynamic Mesh Object Reference
   ) -> (then, Leg_Vertices:Array of Integers)
   > Set right_leg_vertices(right_leg_vertices:Array of Integers=@n2) -> (Output_Get:Array of Integers)
   n3 = SelectMeshElementsInBox!(
    execute
    TargetMesh:Dynamic Mesh Object Reference=Get Modified_Body_Mesh() -> (Modified_Body_Mesh:Dynamic Mesh Object Reference)
    Box:Box Structure=Get Box_Left() -> (Box_Left:Box Structure)
    SelectionType:EGeometryScriptMeshSelectionType Enum=Vertices
    bInvert:bool=false
    MinNumTrianglePoints:int=3
   ) -> (then, Selection:Mesh Selection Structure, ReturnValue:Dynamic Mesh Object Reference)
   > n4 = ConvertMeshSelectionToIndexArray!(execute, TargetMesh:Dynamic Mesh Object Reference=@n3.ReturnValue, Selection:Mesh Selection Structure=@n3) -> (then, IndexArray:Array of Integers, SelectionType:EGeometryScriptMeshSelectionType Enum=Vertices, ReturnValue:Dynamic Mesh Object Reference)
   Cobblers_CreateAltSkinWeights() {is=title K2Node_FunctionEntry}
   > n5 = Sequence!(execute) -> (then_0, then_1, then_2:Exec, then_3:Exec, then_4:Exec, then_5:Exec, then_6:Exec, then_7:Exec, then_8:Exec, then_9:Exec)
   > n6 = For Each Loop(Array=Get Bounds_List() -> (Bounds_List:Array of Strings))
   > n7 = GetAllLevelActors(self=Get EditorActorSubsystem())
   > n8 = For Each Loop with Break!(Exec:Exec, Array:Array of Actor Object References=@n7, Break:Exec=[@n13.then, @n16.then]) -> (LoopBody:Exec, Array Element:Actor Object Reference, Array Index:int, Completed:Exec)
   > Branch(Condition="Equal, Case Insensitive (String)"(A=GetDisplayName(Object=@n8), B=@n6))
   > n9 = Branch(Condition="Equal (Integer)"(A=@n6.Array Index))
   > n10 = Cobblers_GetMeshBounds!(execute, self, Bounds_Object:Object Reference=@n8) -> (then, Box:Dynamic Mesh Object Reference, Box_Xform:Transform)
   > n11 = Set Box_Left(Box_Left:Box Structure=GetMeshBoundingBox!(TargetMesh:Dynamic Mesh Object Reference=@n10) -> (ReturnValue:Box Structure)) -> (Output_Get:Box Structure)
   n12 = Get Box_Array() -> (Box_Array:Array of Box Structures)
   > Add(TargetArray:Array of Box Structures=@n12, NewItem:"Box Structure (by ref)"=@n11) {is=title K2Node_CallArrayFunction}
   > n13 = PrintString(InString=Append(A="Box Left: ", B="To String (Transform)"!(InTrans:"Transform (by ref)"=@n10.Box_Xform) -> (ReturnValue:String)))
   Branch(execute=@n9.else, Condition="Equal (Integer)"(A=@n6.Array Index, B=1))
   > n14 = Cobblers_GetMeshBounds(Bounds_Object=@n8)
   > n15 = Set Box_Right(Box_Right:Box Structure=GetMeshBoundingBox(TargetMesh=@n14)) -> (Output_Get:Box Structure)
   > Add(TargetArray=@n12, NewItem=@n15)
   > n16 = PrintString(InString=Append(A="Box Right: ", B="To String (Transform)"(InTrans=@n14.Box_Xform)))
   n17 = GetSelectionSet!(execute=@n5.then_1) -> (then, ReturnValue:Array of Actor Object References)
   > n18 = Cast To SkeletalMeshActor!(execute, Object="Get (a copy)"(Array:Array of Actor Object References=@n17) -> (Output:Actor Object Reference)) -> (then, CastFailed, AsSkeletal Mesh Actor:Skeletal Mesh Actor Object Reference) {is=title K2Node_DynamicCast}
   > Set Mesh_Transform(Mesh_Transform:Transform=Get Actor Transform(self=@n18)) -> (Output_Get:Transform)
   n19 = Get SkeletalMeshComponent!(self:Skeletal Mesh Actor Object Reference=@n18) -> (SkeletalMeshComponent:Skeletal Mesh Component Object Reference)
   > Set Body_SKMesh_Component(Body_SKMesh_Component:Skeletal Mesh Component Object Reference=@n19) -> (Output_Get:Skeletal Mesh Component Object Reference)
   n20 = Get SkeletalMeshAsset!(self:Skeletal Mesh Component Object Reference=@n19) -> (SkeletalMeshAsset:Skeletal Mesh Object Reference)
   > Set Body_SKM_Asset(Body_SKM_Asset:Skeletal Mesh Object Reference=@n20) -> (Output_Get:Skeletal Mesh Object Reference)
   > Set Body_Skeleton(Body_Skeleton:Skeleton Object Reference=Get Skeleton!(self:Skeletal Mesh Object Reference=@n20) -> (Skeleton:Skeleton Object Reference)) -> (Output_Get:Skeleton Object Reference)
   n21 = Construct Object from Class(execute=@n5.then_2, Class=/Script/GeometryFramework.DynamicMesh) -> (ReturnValue:Dynamic Mesh Object Reference)
   > n22 = CopyMeshFromSkeletalMesh!(
    execute
    FromSkeletalMeshAsset:Skeletal Mesh Object Reference=Get Body_SKM_Asset() -> (Body_SKM_Asset:Skeletal Mesh Object Reference)
    ToDynamicMesh:Dynamic Mesh Object Reference=@n21
    AssetOptions:Geometry Script Copy Mesh From Asset Options Structure
    RequestedLOD:Geometry Script Mesh Read LOD Structure
   ) -> (Failure:Exec, Success:Exec, ReturnValue:Dynamic Mesh Object Reference)
   n23 = TransformMesh!(
    execute=@n22.Success
    TargetMesh:Dynamic Mesh Object Reference=@n22
    Transform:Transform=Get Mesh_Transform() -> (Mesh_Transform:Transform)
    bFixOrientationForNegativeScale:bool=true
   ) -> (then, ReturnValue:Dynamic Mesh Object Reference)
   > n24 = Set UnModified_Body_Mesh(UnModified_Body_Mesh:Dynamic Mesh Object Reference=@n23) -> (Output_Get:Dynamic Mesh Object Reference)
   n25 = Get AltSkinWeightName() -> (AltSkinWeightName:Name=None)
   > n26 = MeshCreateBoneWeights!(execute, TargetMesh:Dynamic Mesh Object Reference=@n24, bReplaceExistingProfile:bool=true, Profile_ProfileName:Name=@n25) -> (then, bProfileExisted:bool, ReturnValue:Dynamic Mesh Object Reference)
   > MeshCopyBoneWeights!(
    execute
    TargetMesh:Dynamic Mesh Object Reference=@n26.ReturnValue
    TargetProfile:Bone Weights Profile Structure=Make Bone Weights Profile!(ProfileName:Name=@n25) -> (GeometryScriptBoneWeightProfile:Bone Weights Profile Structure) {is=title K2Node_MakeStruct}
    SourceProfile:Bone Weights Profile Structure=(ProfileName="Default")
   ) -> (then, bProfileExisted:bool, ReturnValue:Dynamic Mesh Object Reference)
   n27 = Construct Object from Class(execute=@n5.then_3, Class=&DynamicMesh)
   > n28 = CopyMeshFromSkeletalMesh(FromSkeletalMeshAsset=Get Body_SKM_Asset(), ToDynamicMesh=@n27)
   n29 = TransformMesh(execute=@n28.Success, TargetMesh=@n28, Transform=Get Mesh_Transform(), bFixOrientationForNegativeScale=true)
   > Set Transfer_Body_Mesh(Transfer_Body_Mesh:Dynamic Mesh Object Reference=@n29) -> (Output_Get:Dynamic Mesh Object Reference)
   n30 = Construct Object from Class(execute=@n5.then_4, Class=&DynamicMesh)
   > n31 = CopyMeshFromSkeletalMesh(FromSkeletalMeshAsset=Get Legs_SKM_Asset() -> (Legs_SKM_Asset:Skeletal Mesh Object Reference), ToDynamicMesh=@n30)
   n32 = TransformMesh(execute=@n31.Success, TargetMesh=@n31, Transform=Get Mesh_Transform(), bFixOrientationForNegativeScale=true)
   > n33 = Set Legs_Template_Mesh(Legs_Template_Mesh:Dynamic Mesh Object Reference=@n32) -> (Output_Get:Dynamic Mesh Object Reference)
   > TransferBoneWeightsFromMesh!(
    execute
    SourceMesh:Dynamic Mesh Object Reference=@n33
    TargetMesh:Dynamic Mesh Object Reference=Get Transfer_Body_Mesh() -> (Transfer_Body_Mesh:Dynamic Mesh Object Reference)
    Options:Geometry Script Transfer Bone Weights Options Structure=Make Geometry Script Transfer Bone Weights Options!(
     TransferMethod:ETransferBoneWeightsMethod Enum=ClosestPointOnSurface
     OutputTargetMeshBones:EOutputTargetMeshBones Enum=TargetBones
     SourceProfile:Bone Weights Profile Structure=(ProfileName="Default")
     TargetProfile:Bone Weights Profile Structure=(ProfileName="Default")
     RadiusPercentage:double=-1
     NormalThreshold:double=-1
     LayeredMeshSupport:bool=True
     NumSmoothingIterations:int=0
     SmoothingStrength:float=0
     InpaintMask:Name=None
    ) -> (GeometryScriptTransferBoneWeightsOptions:Geometry Script Transfer Bone Weights Options Structure) {is=title K2Node_MakeStruct}
    Selection:Mesh Selection Structure=()
   ) -> (then, ReturnValue:Dynamic Mesh Object Reference)
   n34 = Construct Object from Class(execute=@n5.then_5, Class=&DynamicMesh)
   > n35 = CopyMeshFromSkeletalMesh(FromSkeletalMeshAsset=Get Body_SKM_Asset(), ToDynamicMesh=@n34)
   n36 = TransformMesh(execute=@n35.Success, TargetMesh=@n35, Transform=Get Mesh_Transform(), bFixOrientationForNegativeScale=true)
   > Set Modified_Body_Mesh(Modified_Body_Mesh:Dynamic Mesh Object Reference=@n36) -> (Output_Get:Dynamic Mesh Object Reference)
   n37 = For Each Loop(Exec=@n5.then_6, Array=Get CalfName_List_Single() -> (CalfName_List_Single:Array of Strings))
   n38 = Sequence!(execute=@n37.Completed) -> (then_0, then_1, then_2)
   n39 = Get CalfTwistName_List() -> (CalfTwistName_List:Array of Strings)
   n40 = Get UnModified_Body_Mesh()
   > n41 = GetBoneInfo!(
    execute
    TargetMesh:Dynamic Mesh Object Reference=@n40
    BoneName:Name=String To Name(InString="Get (a copy)"(Array:Array of Strings=@n39) -> (Output:String))
   ) -> (then, bIsValidBoneName:bool, BoneInfo:Bone Info Structure, ReturnValue:Dynamic Mesh Object Reference)
   > Set calf_twist_l_index(calf_twist_l_index:int=Break Bone Info!(GeometryScriptBoneInfo:"Bone Info Structure (by ref)"=@n41.BoneInfo) -> (Index:int, Name:Name, ParentIndex:int, LocalTransform:Transform, WorldTransform:Transform, Color:Linear Color Structure) {is=title K2Node_BreakStruct}) -> (Output_Get:int=0)
   n42 = GetBoneInfo(execute=@n38.then_1, TargetMesh=@n40, BoneName=String To Name(InString="Get (a copy)"(Array=@n39, Dimension 1=1)))
   > Set calf_twist_r_index(calf_twist_r_index:int=Break Bone Info(GeometryScriptBoneInfo=@n42.BoneInfo)) -> (Output_Get:int=0)
   n43 = Break Box!(Box:"Box Structure (by ref)"=Get Box_Left()) -> (Min:Vector=~, Max:Vector=~, IsValid:bool=~) {is=title K2Node_BreakStruct}
   Set Collapse_Location(
    execute=@n38.then_2
    Collapse_Location:Vector=LinePlaneIntersection!(
     LineStart:"Vector (by ref)"=Get Calf_Location() -> (Calf_Location:Vector)
     LineEnd:"Vector (by ref)"=Get Foot_Location() -> (Foot_Location:Vector)
     APlane:"Plane Structure (by ref)"=MakePlaneFromPointAndNormal!(Point:Vector=@n43.Max, Normal:Vector="0,0,1") -> (ReturnValue:Plane Structure)
    ) -> (T:float, Intersection:Vector, ReturnValue:bool).Intersection
   ) -> (Output_Get:Vector)
   > n44 = Cobblers_GetLegVertices(
    Calf_Name="Get (a copy)"(Array=Get CalfName_List() -> (CalfName_List:Array of Strings))
    Thigh_Name="Get (a copy)"(Array=Get ThighName_List())
    Mesh=Get UnModified_Body_Mesh()
   )
   > Set left_leg_vertices(left_leg_vertices:Array of Integers=@n44) -> (Output_Get:Array of Integers)
   > n45 = GetAllBonesInfo!(execute, TargetMesh:Dynamic Mesh Object Reference=Get UnModified_Body_Mesh()) -> (then, BonesInfo:Array of Bone Info Structures, ReturnValue:Dynamic Mesh Object Reference)
   > n46 = For Each Loop(Array:Array of Bone Info Structures=@n45) -> (Array Element:Bone Info Structure)
   n47 = Break Bone Info(GeometryScriptBoneInfo=@n46)
   > LogString(InString=Append!(A="To String (Integer)"(InInt=@n47), B=" :Name: ", C:String="To String (Name)"(InName=@n47.Name)) -> (ReturnValue))
   n48 = Sequence!(execute=@n37.LoopBody) -> (then_0, then_1, then_2, then_3:Exec)
   n49 = Get UnModified_Body_Mesh()
   > n50 = GetBoneInfo(TargetMesh=@n49, BoneName=String To Name(InString=@n37))
   > Set Calf_Location(Calf_Location:Vector=BreakTransform(InTransform=Break Bone Info(GeometryScriptBoneInfo=@n50.BoneInfo).WorldTransform)) -> (Output_Get:Vector)
   n51 = Get FootName_List() -> (FootName_List:Array of Strings)
   n52 = GetBoneInfo(execute=@n48.then_1, TargetMesh=@n49, BoneName=String To Name(InString="Get (a copy)"(Array=@n51, Dimension 1=@n37.Array Index)))
   > Set Foot_Location(Foot_Location:Vector=BreakTransform(InTransform=Break Bone Info(GeometryScriptBoneInfo=@n52.BoneInfo).WorldTransform)) -> (Output_Get:Vector)
   n53 = For Each Loop(Exec=@n5.then_7, Array:Array of Integers=Get left_leg_vertices() -> (left_leg_vertices:Array of Integers)) -> (Array Element:int)
   > SetVertexPosition!(
    execute
    TargetMesh:Dynamic Mesh Object Reference=Get Modified_Body_Mesh()
    VertexID:int=@n53
    NewPosition:Vector=Get Collapse_Location() -> (Collapse_Location:Vector)
    bDeferChangeNotifications:bool=false
   ) -> (then, bIsValidVertex:bool, ReturnValue:Dynamic Mesh Object Reference)
   n54 = For Each Loop(Exec=@n5.then_8, Array=Get left_leg_vertices())
   > n55 = Set vertex(vertex:int=@n54) -> (Output_Get:int=0)
   > n56 = GetVertexBoneWeights!(
    execute
    TargetMesh:Dynamic Mesh Object Reference=Get Transfer_Body_Mesh()
    VertexID:int=Get vertex() -> (vertex:int=0)
    Profile:Bone Weights Profile Structure=(ProfileName="Default")
   ) -> (then, BoneWeights:Array of Bone Weights Structures, bHasValidBoneWeights:bool, ReturnValue:Dynamic Mesh Object Reference)
   > n57 = GetVertexBoneWeights(TargetMesh=Get UnModified_Body_Mesh(), VertexID=Get vertex() -> (vertex=0), Profile=(ProfileName="Default"))
   n58 = BreakVector(InVec=Get Collapse_Location())
   n59 = MakeVector(X=@n58, Y=@n58.Y, Z=float - float(A=@n58.Z, B=3))
   n60 = Get Calf_Location()
   n61 = GetVertexPosition!(TargetMesh:Dynamic Mesh Object Reference=Get UnModified_Body_Mesh(), VertexID:int=@n55) -> (bIsValidVertex:bool, ReturnValue:Vector)
   > SetVertexBoneWeights!(
    execute
    TargetMesh:Dynamic Mesh Object Reference=Get UnModified_Body_Mesh()
    VertexID:int=Get vertex() -> (vertex=0)
    BoneWeights:Array of Bone Weights Structures=BlendBoneWeights!(
     BoneWeightsA:Array of Bone Weights Structures=@n57
     BoneWeightsB:Array of Bone Weights Structures=@n56
     Alpha:float=Get Point Closest to Segment Alpha!(A:Vector=@n60, B:Vector=@n59, Point:Vector=@n61.ReturnValue) -> (Alpha:double=~) {is=title K2Node_MacroInstance}
    ) -> (Result:Array of Bone Weights Structures)
    Profile:Bone Weights Profile Structure=Make Bone Weights Profile(ProfileName=Get AltSkinWeightName() -> (AltSkinWeightName=None))
   ) -> (then, bIsValidVertexID:bool, ReturnValue:Dynamic Mesh Object Reference)
   n62 = For Each Loop(Exec=@n54.Completed, Array=Get left_leg_vertices())
   n63 = GetVertexPosition(TargetMesh=Get Modified_Body_Mesh(), VertexID=@n62)
   > n64 = Branch(Condition="float > float"(A=BreakVector(InVec=@n63.ReturnValue)))
   n65 = Get calf_twist_l_index() -> (calf_twist_l_index:int=0)
   n66 = Make Bone Weights Profile(ProfileName=Get AltSkinWeightName() -> (AltSkinWeightName=None))
   n67 = Get UnModified_Body_Mesh()
   > SetVertexBoneWeights(
    TargetMesh=@n67
    VertexID=@n62
    BoneWeights=Make Array("[0]":Bone Weights Structure=Make Bone Weights!(BoneIndex:int=@n65, Weight:float=1) -> (GeometryScriptBoneWeight:Bone Weights Structure) {is=title K2Node_MakeStruct}) -> (Array:Array of Bone Weights Structures)
    Profile=@n66
   )
   n68 = Get calf_twist_r_index() -> (calf_twist_r_index:int=0)
   SetVertexBoneWeights(execute=@n64.else, TargetMesh=@n67, VertexID=@n62, BoneWeights=Make Array("[0]"=Make Bone Weights(BoneIndex=@n68, Weight=1)), Profile=@n66)
   n69 = Sequence!(execute=@n5.then_9) -> (then_0, then_1, then_2)
   > CreateNewSkeletalMeshAssetFromMesh!(
    execute
    FromDynamicMesh:Dynamic Mesh Object Reference=Get UnModified_Body_Mesh()
    InSkeleton:Skeleton Object Reference=Get Body_Skeleton() -> (Body_Skeleton:Skeleton Object Reference)
    AssetPathAndName:String=/Game/NeverCook/ArtTools/Tech_Anim/CobblerGeoScripts/Meshes/Body_with_AltSkinWeights
    Options:Geometry Script Create New Skeletal Mesh Asset Options Structure
   ) -> (Failure:Exec, Success:Exec, ReturnValue:Skeletal Mesh Object Reference)
   CreateNewSkeletalMeshAssetFromMesh(
    execute=@n69.then_1
    FromDynamicMesh=Get Transfer_Body_Mesh()
    InSkeleton=Get Body_Skeleton()
    AssetPathAndName=/Game/NeverCook/ArtTools/Tech_Anim/CobblerGeoScripts/Meshes/Body_with_LegWeights
   )
   CreateNewSkeletalMeshAssetFromMesh(
    execute=@n69.then_2
    FromDynamicMesh=Get Modified_Body_Mesh()
    InSkeleton=Get Body_Skeleton()
    AssetPathAndName=/Game/NeverCook/ArtTools/Tech_Anim/CobblerGeoScripts/Meshes/Body_Modified
   )
   n70 = For Each Loop(Array:Array of Bone Weights Structures) -> (Array Element:Bone Weights Structure)
   n71 = Break Bone Weights!(GeometryScriptBoneWeight:"Bone Weights Structure (by ref)"=@n70) -> (BoneIndex:int, Weight:float=~) {is=title K2Node_BreakStruct}
   n72 = Get vertex() -> (vertex=0)
   > LogString(InString=Append(A="To String (Integer)"(InInt=@n71), B=" :Vertex: ", C="To String (Integer)"(InInt=@n72)))
   n73 = Get Calf_Location()
   n74 = BreakVector(InVec=@n73)
   MakeVector(X=@n74, Y=@n74.Y, Z=float - float(A=@n74.Z, B=float * float(A=@n74.Z, B=0.2) {is=title K2Node_PromotableOperator}))
   SetVertexPosition(TargetMesh=Get Modified_Body_Mesh(), VertexID=Get vertex() -> (vertex=0), NewPosition="0, 0, 0", bDeferChangeNotifications=false)
   "To String Center and Extents (Box)"!(Box:"Box Structure (by ref)"=@n11) -> (ReturnValue:String)
   "To String Center and Extents (Box)"(Box=@n15)
   Branch(Condition=Point in Capsule!(A:Vector=@n73, B:Vector=Get Foot_Location(), Radius:double=Get Leg_Radius() -> (Leg_Radius:double), Point:Vector=@n61.ReturnValue) -> (IsInside:bool=~) {is=title K2Node_MacroInstance})
   For Each Loop(Array:Array of Box Structures=Get Box_Array()) -> (Array Element:Box Structure)
   Get Foot_Location()
   Get vertex_id() -> (vertex_id:int=0)
   Get CalfName_List()
   Get Modified_Vertex_Array() -> (Modified_Vertex_Array:Array of Integers)
   Set Modified_Vertex_Array(Modified_Vertex_Array:Array of Integers=@n4) -> (Output_Get:Array of Integers)
  ]}
 Cobblers_CreateLegCapsule {nodes [Cobblers_CreateLegCapsule() {is=title K2Node_FunctionEntry}]}
 Cobblers_CreateMorphTarget {nodes [
   n1 = Make Index List!(IndexType:EGeometryScriptIndexType Enum=Triangle) -> (GeometryScriptIndexList:Index List Structure) {is=title K2Node_MakeStruct}
   n2 = Get Triangles In PolyGroup!(
    execute
    TargetMesh:Dynamic Mesh Object Reference
    GroupLayer:Geometry Script Group Layer Structure=Make Geometry Script Group Layer!(bDefaultLayer:bool=True, ExtendedLayerIndex:int=0) -> (GeometryScriptGroupLayer:Geometry Script Group Layer Structure) {is=title K2Node_MakeStruct}
    PolygroupID:int=0
    TriangleIDsOut:"Index List Structure (by ref)"=@n1
   ) -> (then, ReturnValue:Dynamic Mesh Object Reference) {is=title K2Node_CallFunction}
   > ConvertIndexListToMeshSelection!(
    execute
    TargetMesh:Dynamic Mesh Object Reference=@n2
    IndexList:Index List Structure=@n1
    SelectionType:EGeometryScriptMeshSelectionType Enum=Triangles
   ) -> (then, Selection:Mesh Selection Structure, ReturnValue:Dynamic Mesh Object Reference)
   n3 = TransferBoneWeightsFromMesh(
    Options=Make Geometry Script Transfer Bone Weights Options(
     TransferMethod=ClosestPointOnSurface
     OutputTargetMeshBones=TargetBones
     SourceProfile=(ProfileName="Default")
     TargetProfile=(ProfileName="Default")
     RadiusPercentage=-1
     NormalThreshold=-1
     LayeredMeshSupport=True
     NumSmoothingIterations=0
     SmoothingStrength=0
     InpaintMask=None
    )
    Selection=()
   )
   > n4 = CreateNewSkeletalMeshAssetFromMesh(
    FromDynamicMesh=@n3
    InSkeleton=/BRCosmetics/Characters/Player/Female/Medium/Bodies/F_MED_AgentSherbert/Meshes/F_MED_AgentSherbert_Skeleton
    AssetPathAndName=/Game/00_TEST/Cobblers/Cobblers_Shoes
   )
   CopyMeshToSkeletalMesh!(
    execute=@n4.Success
    FromDynamicMesh:Dynamic Mesh Object Reference=@n3
    ToSkeletalMeshAsset:Skeletal Mesh Object Reference
    Options:Geometry Script Copy Mesh To Asset Options Structure
    TargetLOD:Geometry Script Mesh Write LOD Structure
   ) -> (Failure:Exec, Success:Exec, ReturnValue:Dynamic Mesh Object Reference)
   Cobblers_CreateMorphTarget() {is=title K2Node_FunctionEntry}
   > n5 = Sequence!(execute) -> (then_0, then_1, then_2, then_3:Exec)
   > n6 = For Each Loop(Array:Array of Strings=Get Bounds_List()) -> (Array Element:String)
   > n7 = GetAllLevelActors(self=Get EditorActorSubsystem())
   > n8 = For Each Loop with Break(Array=@n7, Break=[@n13.then, @n16.then])
   > Branch(Condition="Equal, Case Insensitive (String)"(A=GetDisplayName(Object=@n8), B=@n6))
   > n9 = Branch(Condition="Equal (Integer)"(A=@n6.Array Index))
   > n10 = Cobblers_GetMeshBounds(Bounds_Object=@n8)
   > n11 = Set Box_Left(Box_Left=GetMeshBoundingBox(TargetMesh=@n10))
   n12 = Get Box_Array()
   > Add(TargetArray=@n12, NewItem=@n11)
   > n13 = PrintString(InString=Append!(A="Box Left: ", B="To String (Transform)"(InTrans=@n10.Box_Xform)) -> (ReturnValue))
   Branch(execute=@n9.else, Condition="Equal (Integer)"(A=@n6.Array Index, B=1))
   > n14 = Cobblers_GetMeshBounds(Bounds_Object=@n8)
   > n15 = Set Box_Right(Box_Right=GetMeshBoundingBox(TargetMesh=@n14))
   > Add(TargetArray=@n12, NewItem=@n15)
   > n16 = PrintString(InString=Append(A="Box Right: ", B="To String (Transform)"(InTrans=@n14.Box_Xform)))
   n17 = For Each Loop(Exec=@n5.then_1, Array=Get Template_List() -> (Template_List:Array of Strings))
   > n18 = GetAllLevelActors(self=Get EditorActorSubsystem())
   > n19 = For Each Loop with Break(Array=@n18, Break=@n24.then)
   > Branch(Condition="Equal, Case Insensitive (String)"(A=GetDisplayName(Object=@n19), B=@n17))
   > n20 = Cast To SkeletalMeshActor(Object=@n19)
   n21 = Get SkeletalMeshComponent(self=@n20)
   > Set Template_SKMesh_Component(Template_SKMesh_Component:Skeletal Mesh Component Object Reference=@n21) -> (Output_Get:Skeletal Mesh Component Object Reference)
   > n22 = Construct Object from Class(Class=&DynamicMesh)
   > n23 = CopyMeshFromSkeletalMesh(FromSkeletalMeshAsset=Get SkeletalMeshAsset(self=@n21), ToDynamicMesh=@n22)
   n24 = Set Template_Leg_Mesh(execute=@n23.Success, Template_Leg_Mesh:Dynamic Mesh Object Reference=@n23) -> (Output_Get:Dynamic Mesh Object Reference)
   n25 = GetSelectionSet(execute=@n5.then_2)
   > n26 = Cast To SkeletalMeshActor(Object="Get (a copy)"(Array:Array of Actor Object References=@n25) -> (Output:Actor Object Reference))
   n27 = Get SkeletalMeshComponent(self=@n26)
   > Set Shoes_SKMesh_Component(Shoes_SKMesh_Component:Skeletal Mesh Component Object Reference=@n27) -> (Output_Get:Skeletal Mesh Component Object Reference)
   > Set Shoes_SKM_Asset(Shoes_SKM_Asset:Skeletal Mesh Object Reference=Get SkeletalMeshAsset(self=@n27)) -> (Output_Get:Skeletal Mesh Object Reference)
   > n28 = Cobblers_GetMaterials!(execute, self, Component:Mesh Component Object Reference=@n27) -> (then, NewParam:Map of Names to Material Interface Object References)
   > Set Shoe_Materials(Shoe_Materials:Map of Names to Material Interface Object References=@n28) -> (Output_Get:Map of Names to Material Interface Object References)
   > n29 = Construct Object from Class(Class=&DynamicMesh)
   > n30 = CopyMeshFromComponent!(
    execute
    Component:Scene Component Object Reference=@n27
    ToDynamicMesh:Dynamic Mesh Object Reference=@n29
    Options:Geometry Script Copy Mesh From Component Options Structure
    bTransformToWorld:bool=true
   ) -> (Failure:Exec, Success:Exec, LocalToWorld:Transform, ReturnValue:Dynamic Mesh Object Reference)
   n31 = CopyMeshFromSkeletalMesh(execute=@n30.Success, FromSkeletalMeshAsset=Get Shoes_SKM_Asset() -> (Shoes_SKM_Asset:Skeletal Mesh Object Reference), ToDynamicMesh=@n29)
   n32 = GetSelectionSet(execute=@n5.then_3)
   > n33 = Cast To SkeletalMeshActor(Object="Get (a copy)"(Array=@n32, Dimension 1=1))
   n34 = Get SkeletalMeshComponent(self=@n33)
   > Set Body_SKMesh_Component(Body_SKMesh_Component=@n34)
   n35 = Get SkeletalMeshAsset(self=@n34)
   > Set Body_SKM_Asset(Body_SKM_Asset=@n35)
   > Set Body_Skeleton(Body_Skeleton=Get Skeleton(self=@n35))
   > n36 = Cobblers_GetMaterials(Component=@n34)
   > Set Body_Materials(Body_Materials:Map of Names to Material Interface Object References=@n36) -> (Output_Get:Map of Names to Material Interface Object References)
   > n37 = Construct Object from Class(Class=&DynamicMesh)
   > n38 = CopyMeshFromSkeletalMesh(FromSkeletalMeshAsset=@n35, ToDynamicMesh=@n37)
   n39 = Get Actor Transform(self=@n33)
   n40 = TransformMesh(execute=@n38.Success, TargetMesh=@n38, Transform=@n39, bFixOrientationForNegativeScale=true)
   > Set Modified_Body_Mesh(Modified_Body_Mesh=@n40)
   > n41 = Construct Object from Class(Class=&DynamicMesh)
   > n42 = CopyMeshFromSkeletalMesh(FromSkeletalMeshAsset=Get Body_SKM_Asset(), ToDynamicMesh=@n41)
   n43 = TransformMesh(execute=@n42.Success, TargetMesh=@n42, Transform=@n39, bFixOrientationForNegativeScale=true)
   > Set UnModified_Body_Mesh(UnModified_Body_Mesh=@n43)
   > n44 = CopyBonesFromMesh!(
    execute
    SourceMesh:Dynamic Mesh Object Reference=Get Modified_Body_Mesh()
    TargetMesh:Dynamic Mesh Object Reference=@n31
    Options:Geometry Script Copy Bones From Mesh Options Structure=Make Geometry Script Copy Bones from Mesh Options!(ReindexWeights:bool=true, BonesToCopyFromSource:EBonesToCopyFromSource Enum=OnlyBoundAndParents) -> (GeometryScriptCopyBonesFromMeshOptions:Geometry Script Copy Bones From Mesh Options Structure) {is=title K2Node_MakeStruct}
   ) -> (then, ReturnValue:Dynamic Mesh Object Reference)
   > n45 = CreateNewSkeletalMeshAssetFromMesh(
    FromDynamicMesh=@n44
    InSkeleton=Get Body_Skeleton()
    AssetPathAndName=/BRCosmetics/Characters/Player/Test/Shoes/Normalized/Test/Rebind_Shoes
    Options=Make Geometry Script Create New Skeletal Mesh Asset Options!(
     bEnableRecomputeNormals:bool=False
     bEnableRecomputeTangents:bool=False
     Materials:Map of Names to Material Interface Object References=Get Shoe_Materials() -> (Shoe_Materials:Map of Names to Material Interface Object References)
     bUseMeshBoneProportions:bool=true
     bApplyNaniteSettings:bool=False
     NaniteSettings:Mesh Nanite Settings Structure=(bEnabled=False,bExplicitTangents=False,bLerpUVs=True,bSeparable=False,bVoxelNDF=True,bVoxelOpacity=False,ShapePreservation=None,PositionPrecision=-2147483648,NormalPrecision=-1,TangentPrecision=-1,BoneWeightPrecision=-1,TargetMinimumResidencyInKB=0,KeepPercentTriangles=1,TrimRelativeError=0,GenerateFallback=PlatformDefault,FallbackTarget=Auto,FallbackPercentTriangles=1,FallbackRelativeError=1,MaxEdgeLengthFactor=0,NumRays=64,VoxelLevel=0,RayBackUp=0,DisplacementUVChannel=0,DisplacementMaps=,NaniteAssemblyData=(Parts=,Nodes=))
     bUseOriginalVertexOrder:bool=False
    ) -> (GeometryScriptCreateNewSkeletalMeshAssetOptions:Geometry Script Create New Skeletal Mesh Asset Options Structure) {is=title K2Node_MakeStruct}
   )
   n46 = Get AltSkinWeightName() -> (AltSkinWeightName=None)
   n47 = MeshCreateBoneWeights(execute=@n45.Success, TargetMesh=Get UnModified_Body_Mesh(), bReplaceExistingProfile=true, Profile_ProfileName=@n46)
   > MeshCopyBoneWeights(TargetMesh=@n47.ReturnValue, TargetProfile=Make Bone Weights Profile(ProfileName=@n46), SourceProfile=(ProfileName="Default"))
   > n48 = For Each Loop(Array:Array of Box Structures=Get Box_Array()) -> (Array Element:Box Structure)
   n49 = Get CalfName_List()
   > n50 = GetBoneInfo(
    TargetMesh=Get Modified_Body_Mesh()
    BoneName=String To Name(InString="Get (a copy)"(Array:Array of Strings=@n49, Dimension 1=@n48.Array Index) -> (Output:String))
   )
   n51 = Get FootName_List()
   > n52 = GetBoneInfo(TargetMesh=Get Modified_Body_Mesh(), BoneName=String To Name(InString="Get (a copy)"(Array=@n51, Dimension 1=@n48.Array Index)))
   > n53 = SelectMeshElementsInBox(TargetMesh=@n50.ReturnValue, Box=@n48, SelectionType=Vertices, bInvert=false, MinNumTrianglePoints=3)
   > Add(
    TargetArray:Array of Mesh Selection Structures=Make Array("[0]":Mesh Selection Structure=Get Mesh_Vertex_Selection() -> (Mesh_Vertex_Selection:Mesh Selection Structure)) -> (Array:Array of Mesh Selection Structures)
    NewItem:"Mesh Selection Structure (by ref)"=@n53
   )
   > n54 = ConvertMeshSelectionToIndexArray(TargetMesh=@n53.ReturnValue, Selection=@n53) -> (SelectionType=Vertices)
   > n55 = For Each Loop(Array:Array of Integers=@n54) -> (Array Element:int)
   n56 = Get Foot_Points() -> (Foot_Points:Set of Integers)
   Clear!(execute=@n55.Completed, TargetSet:Set of Integers=@n56) -> (then) {is=title K2Node_CallFunction}
   > Add Items!(execute, TargetSet:Set of Integers=@n56, NewItems:Array of Integers=@n54) -> (then)
   > n57 = GetAllVertexIDs(TargetMesh=@n54.ReturnValue)
   > n58 = ConvertIndexListToArray(IndexList=@n57)
   > n59 = For Each Loop(Array=@n58)
   > n60 = Branch(Condition=Contains Item!(TargetSet:Set of Integers=Get Foot_Points(), ItemToFind:"Integer (by ref)"=@n59) -> (ReturnValue:bool) {is=title K2Node_CallFunction})
   n61 = Get UnModified_Body_Mesh()
   n62 = Make Bone Weights Profile(ProfileName=Get AltSkinWeightName() -> (AltSkinWeightName=None))
   n63 = Break Bone Info(GeometryScriptBoneInfo=@n50.BoneInfo)
   n64 = Make Array("[0]":Bone Weights Structure=Make Bone Weights(BoneIndex=@n63, Weight=1)) -> (Array:Array of Bone Weights Structures)
   > SetVertexBoneWeights(TargetMesh=@n61, VertexID=@n59, BoneWeights=@n64, Profile=@n62)
   n65 = GetVertexPosition(TargetMesh=@n61, VertexID=@n59)
   n66 = BreakTransform(InTransform=@n63.WorldTransform)
   n67 = BreakTransform(InTransform=Break Bone Info(GeometryScriptBoneInfo=@n52.BoneInfo).WorldTransform)
   Branch(execute=@n60.else, Condition=Point in Capsule(A=@n66, B=@n67, Radius=Get Leg_Radius(), Point=@n65.ReturnValue) -> (IsInside=~))
   > n68 = GetVertexBoneWeights(TargetMesh=@n61, VertexID=@n59, Profile=@n62)
   n69 = LinePlaneIntersection(LineStart=@n66, LineEnd=@n67, APlane=MakePlaneFromPointAndNormal(Point=Break Box(Box=@n48) -> (Min=~, Max=~, IsValid=~).Max, Normal="0,0,1"))
   > SetVertexBoneWeights(
    TargetMesh=@n61
    VertexID=@n59
    BoneWeights=BlendBoneWeights(
     BoneWeightsA=@n68
     BoneWeightsB=@n64
     Alpha=Get Point Closest to Segment Alpha(A=@n66, B=@n69.Intersection, Point=@n65.ReturnValue) -> (Alpha=~)
    )
    Profile=@n62
   )
   SetVertexPosition(execute=@n55.LoopBody, TargetMesh=@n54.ReturnValue, VertexID=@n55, NewPosition=@n69.Intersection, bDeferChangeNotifications=false)
   n70 = CopyMorphTargetToSkeletalMesh!(
    execute=@n48.Completed
    FromMorphTarget:Dynamic Mesh Object Reference=Get Modified_Body_Mesh()
    ToSkeletalMeshAsset:Skeletal Mesh Object Reference=Get Body_SKM_Asset()
    MorphTargetName:Name=shoes_mid
    Options:Geometry Script Copy Morph Target To Asset Options Structure=Make Geometry Script Copy Morph Target to Asset Options!(bOverwriteExistingTarget:bool=true, bEmitTransaction:bool=false, bDeferMeshPostEditChange:bool=False, bCopyNormals:bool=False) -> (GeometryScriptCopyMorphTargetToAssetOptions:Geometry Script Copy Morph Target To Asset Options Structure) {is=title K2Node_MakeStruct}
    TargetLOD:Geometry Script Mesh Write LOD Structure
   ) -> (Failure:Exec, Success:Exec, ReturnValue:Dynamic Mesh Object Reference)
   n71 = CopyMeshToSkeletalMesh(execute=@n70.Success, FromDynamicMesh=Get UnModified_Body_Mesh(), ToSkeletalMeshAsset=Get Body_SKM_Asset())
   CreateNewSkeletalMeshAssetFromMesh(
    execute=@n71.Success
    FromDynamicMesh=@n71
    InSkeleton=Get Body_Skeleton()
    AssetPathAndName=/BRCosmetics/Characters/Player/Test/Shoes/Normalized/Test/Body_with_SkinWeightProfile
    Options=Make Geometry Script Create New Skeletal Mesh Asset Options(
     bEnableRecomputeNormals=False
     bEnableRecomputeTangents=False
     Materials=Get Body_Materials() -> (Body_Materials:Map of Names to Material Interface Object References)
     bUseMeshBoneProportions=False
     bApplyNaniteSettings=False
     NaniteSettings=(bEnabled=False,bExplicitTangents=False,bLerpUVs=True,bSeparable=False,bVoxelNDF=True,bVoxelOpacity=False,ShapePreservation=None,PositionPrecision=-2147483648,NormalPrecision=-1,TangentPrecision=-1,BoneWeightPrecision=-1,TargetMinimumResidencyInKB=0,KeepPercentTriangles=1,TrimRelativeError=0,GenerateFallback=PlatformDefault,FallbackTarget=Auto,FallbackPercentTriangles=1,FallbackRelativeError=1,MaxEdgeLengthFactor=0,NumRays=64,VoxelLevel=0,RayBackUp=0,DisplacementUVChannel=0,DisplacementMaps=,NaniteAssemblyData=(Parts=,Nodes=))
     bUseOriginalVertexOrder=False
    )
   )
   > PrintText(InText="INVTEXT(\"Why not working?\")")
   n72 = Set Points(Points:Array of Vectors=Make Array!("[0]":Vector=MakeVector(X=5, Y=1, Z=-25), "[1]":Vector=MakeVector(X=25, Y=30, Z=30)) -> (Array:Array of Vectors)) -> (Output_Get:Array of Vectors)
   > n73 = MakeBoundingBoxFromPoints!(execute, Points:Array of Vectors=@n72, ExpandBy:double=0) -> (then, ReturnValue:Box Structure)
   > Set Left_Bounding_Box(Left_Bounding_Box:Box Structure=@n73) -> (Output_Get:Box Structure)
   n74 = Break Box(Box=GetMeshBoundingBox()) -> (Min=~, Max=~, IsValid=~)
   BreakVector(InVec=@n74.Max)
   SaveAsset!(
    execute
    self:Editor Asset Subsystem Object Reference=Get EditorAssetSubsystem() -> (ReturnValue:Editor Asset Subsystem Object Reference) {is=title K2Node_GetEditorSubsystem}
    AssetToSave:String
    bOnlyIfIsDirty:bool=true
   ) -> (then, ReturnValue:bool)
   TransferBoneWeightsFromMesh(
    Options=Make Geometry Script Transfer Bone Weights Options(
     TransferMethod=ClosestPointOnSurface
     OutputTargetMeshBones=SourceBones
     SourceProfile=(ProfileName="Default")
     TargetProfile=(ProfileName="Default")
     RadiusPercentage=-1
     NormalThreshold=-1
     LayeredMeshSupport=True
     NumSmoothingIterations=0
     SmoothingStrength=0
     InpaintMask=None
    )
    Selection=()
   )
   BreakVector(InVec=@n74)
   "To String Center and Extents (Box)"(Box=@n11)
   "To String Center and Extents (Box)"(Box=@n15)
   Reroute Node(InputPin:int=@n55) -> (OutputPin:int)
   Reroute Node(InputPin:Dynamic Mesh Object Reference=@n40) -> (OutputPin:Dynamic Mesh Object Reference)
   Get Modified_Body_Mesh()
  ]}
 Cobblers_CreateShoeMesh {nodes [
   Cobblers_CreateShoeMesh() {is=title K2Node_FunctionEntry}
   > n1 = Sequence!(execute) -> (then_0, then_1)
   > n2 = Construct Object from Class(Class=&DynamicMesh, self=Get Master_SKM_Asset() -> (Master_SKM_Asset:Skeletal Mesh Object Reference))
   n3 = Construct Object from Class(execute=@n1.then_1, Class=&DynamicMesh, self=Get Shoes_SKM_Asset())
   > CopyBonesFromMesh(
    SourceMesh=@n2
    TargetMesh=@n3
    Options=Make Geometry Script Copy Bones from Mesh Options(ReindexWeights=true, BonesToCopyFromSource=OnlyBoundAndParents)
   )
  ]}
 Cobblers_GetLegVertices {nodes [
   n1 = Cobblers_GetLegVertices!() -> (then, Calf_Name:String, Thigh_Name:String, Mesh:Dynamic Mesh Object Reference) {is=title K2Node_FunctionEntry}
   > Set Body_Mesh(Body_Mesh:Dynamic Mesh Object Reference=@n1.Mesh) -> (Output_Get:Dynamic Mesh Object Reference)
   > Set calf_bone_name(calf_bone_name:String=@n1) -> (Output_Get:String)
   > Set thigh_bone_name(thigh_bone_name:String=@n1.Thigh_Name) -> (Output_Get:String)
   > n2 = Sequence!(execute) -> (then_0, then_1, then_2:Exec)
   > n3 = GetAllBonesInfo(TargetMesh=Get Body_Mesh() -> (Body_Mesh:Dynamic Mesh Object Reference))
   > n4 = Sequence!(execute) -> (then_0, then_1)
   > n5 = For Each Loop with Break(Array:Array of Bone Info Structures=@n3, Break=@n8.then) -> (Array Element:Bone Info Structure)
   n6 = Break Bone Info(GeometryScriptBoneInfo=@n5)
   > Branch(Condition="Equal Exactly (String)"(A="To String (Name)"(InName=@n6.Name), B=Get thigh_bone_name() -> (thigh_bone_name:String)))
   > n7 = Set thigh_bone_index(thigh_bone_index:int=@n6) -> (Output_Get:int=0)
   > n8 = LogString(InString=Append(A="Thigh Bone Index: ", B="To String (Integer)"(InInt=@n7)))
   n9 = For Each Loop with Break(Exec=@n4.then_1, Array=@n3, Break=@n13.Completed)
   n10 = Break Bone Info(GeometryScriptBoneInfo=@n9)
   n11 = "To String (Name)"(InName=@n10.Name)
   > Branch(Condition="Equal Exactly (String)"(A=Get calf_bone_name() -> (calf_bone_name:String), B=@n11))
   > LogString(
    InString=Append!(A="Adding Bone Index: ", B="To String (Integer)"(InInt=@n10), C:String=": ", D:String=@n11) -> (ReturnValue)
    bPrintToLog=false
   )
   > Set calf_bone_indices(calf_bone_indices:Array of Integers=Make Array!("[0]":int=@n10) -> (Array:Array of Integers)) -> (Output_Get:Array of Integers)
   > n12 = GetBoneChildren!(execute, TargetMesh:Dynamic Mesh Object Reference=Get Body_Mesh(), BoneName:Name=@n10.Name, bRecursive:bool=true) -> (then, bIsValidBoneName:bool, ChildrenInfo:Array of Bone Info Structures, ReturnValue:Dynamic Mesh Object Reference)
   > n13 = For Each Loop(Array:Array of Bone Info Structures=@n12.ChildrenInfo) -> (Array Element:Bone Info Structure)
   n14 = Break Bone Info(GeometryScriptBoneInfo=@n13)
   > Add Unique(TargetArray:Array of Integers=Get calf_bone_indices() -> (calf_bone_indices:Array of Integers), NewItem:"Integer (by ref)"=@n14)
   > LogString(InString=Append(A="Adding Bone Index: ", B="To String (Integer)"(InInt=@n14), C=": ", D="To String (Name)"(InName=@n14.Name)))
   n15 = Set valid_vertices(execute=@n2.then_1, valid_vertices:Array of Integers=Make Array("[0]"=0)) -> (Output_Get:Array of Integers)
   > Clear(TargetArray:Array of Integers=@n15) {is=title K2Node_CallArrayFunction}
   > n16 = GetAllVertexIDs(TargetMesh=Get Body_Mesh())
   > n17 = ConvertIndexListToArray(IndexList=@n16)
   > n18 = For Each Loop with Break(Array:Array of Integers=@n17) -> (Array Element:int)
   > n19 = Set current_vert_index(current_vert_index:int=@n18) -> (Output_Get:int=0)
   > n20 = GetVertexBoneWeights(TargetMesh=Get Body_Mesh(), VertexID=@n19, Profile=(ProfileName="Default"))
   > n21 = Set current_bone_indices(current_bone_indices:Array of Integers=Make Array("[0]"=0)) -> (Output_Get:Array of Integers)
   > Clear(TargetArray=@n21)
   > n22 = For Each Loop(Array:Array of Bone Weights Structures=@n20) -> (Array Element:Bone Weights Structure)
   > n23 = Set current_bone_index(current_bone_index:int=Break Bone Weights(GeometryScriptBoneWeight=@n22) -> (Weight=~)) -> (Output_Get:int=0)
   > n24 = Sequence()
   > Branch(Condition=Contains Item(TargetArray:Array of Integers=Get calf_bone_indices(), ItemToFind:"Integer (by ref)"=Get current_bone_index() -> (current_bone_index:int=0)) {is=title K2Node_CallArrayFunction})
   n25 = Get current_bone_indices() -> (current_bone_indices:Array of Integers)
   > Add Unique(TargetArray=@n25, NewItem=Get current_bone_index() -> (current_bone_index=0))
   Branch(execute=@n24.then_1, Condition="Equal (Integer)"(A=@n23, B=Get thigh_bone_index() -> (thigh_bone_index:int=0)))
   > Add Unique(TargetArray=@n25, NewItem=Get current_bone_index() -> (current_bone_index=0))
   n26 = Branch(
    execute=@n22.Completed
    Condition=Contains Item(TargetArray=Get current_bone_indices(), ItemToFind=Get thigh_bone_index() -> (thigh_bone_index=0))
   )
   Branch(execute=@n26.else, Condition=Is Not Empty!(TargetArray:Array of Integers=Get current_bone_indices()) -> (ReturnValue:bool))
   n27 = Get current_vert_index() -> (current_vert_index:int=0)
   > Add Unique(TargetArray=Get valid_vertices() -> (valid_vertices:Array of Integers), NewItem=@n27)
   > LogString(InString=Append!(A="Adding Vertex Index: ", B="To String (Integer)"(InInt=@n27)) -> (ReturnValue))
   > n28 = For Each Loop(Array:Array of Integers=Get current_bone_indices()) -> (Array Element:int)
   > LogString(InString=Append(A=" Bone Index: ", B="To String (Integer)"(InInt=@n28)))
   Return Node!(execute=@n2.then_2, Leg_Vertices:Array of Integers=Get valid_vertices())
  ]}
 Cobblers_GetMaterials {nodes [
   n1 = Cobblers_GetMaterials!() -> (then, Component:Mesh Component Object Reference) {is=title K2Node_FunctionEntry}
   > Set Mesh_Component(Mesh_Component:Mesh Component Object Reference=@n1) -> (Output_Get:Mesh Component Object Reference)
   > n2 = For Each Loop(Array:Array of Names=GetMaterialSlotNames!(self:Primitive Component Object Reference=Get Mesh_Component() -> (Mesh_Component:Mesh Component Object Reference)) -> (ReturnValue:Array of Names)) -> (Array Element:Name)
   > Add!(
    execute
    TargetMap:Map of Names to Material Interface Object References=Get Materials() -> (Materials:Map of Names to Material Interface Object References)
    Key:"Name (by ref)"=@n2
    Value:"Material Interface Object Reference (by ref)"=GetMaterialByName!(self:Primitive Component Object Reference=Get Mesh_Component(), MaterialSlotName:Name=@n2) -> (ReturnValue:Material Interface Object Reference)
   ) -> (then) {is=title K2Node_CallFunction}
   Return Node!(execute=@n2.Completed, NewParam:Map of Names to Material Interface Object References=Get Materials())
  ]}
 Cobblers_GetMeshBounds {nodes [
   n1 = Cobblers_GetMeshBounds!() -> (then, Bounds_Object:Object Reference) {is=title K2Node_FunctionEntry}
   > n2 = Cast To StaticMeshActor!(execute, Object=@n1) -> (then, CastFailed, AsStatic Mesh Actor:Static Mesh Actor Object Reference) {is=title K2Node_DynamicCast}
   > n3 = Construct Object from Class(Class=&DynamicMesh)
   > n4 = Copy Mesh From Static Mesh with Section Materials!(
    execute
    FromStaticMeshAsset:Static Mesh Object Reference=Get StaticMesh(self=Get StaticMeshComponent(self=@n2))
    ToDynamicMesh:Dynamic Mesh Object Reference=@n3
    AssetOptions:Geometry Script Copy Mesh From Asset Options Structure
    RequestedLOD:Geometry Script Mesh Read LOD Structure
   ) -> (Failure:Exec, Success:Exec, ReturnValue:Dynamic Mesh Object Reference)
   n5 = Get Actor Transform(self=@n2)
   n6 = TransformMesh(execute=@n4.Success, TargetMesh=@n4, Transform=@n5, bFixOrientationForNegativeScale=true)
   > Return Node!(execute, Box:Dynamic Mesh Object Reference=@n6, Box_Xform:Transform=@n5)
  ]}
 Cobblers_PolygroupSelection {nodes [
   Cobblers_PolygroupSelection() {is=title K2Node_FunctionEntry}
   > n1 = GetSelectionSet()
   > n2 = Cast To SkeletalMeshActor(Object="Get (a copy)"(Array:Array of Actor Object References=@n1) -> (Output:Actor Object Reference))
   > n3 = Construct Object from Class(Class=&DynamicMesh)
   n4 = Get SkeletalMeshAsset(self=Get SkeletalMeshComponent(self=@n2))
   > n5 = CopyMeshFromSkeletalMesh(FromSkeletalMeshAsset=@n4, ToDynamicMesh=@n3)
   n6 = Make Index List(IndexType=Triangle)
   n7 = Get Triangles In PolyGroup(execute=@n5.Success, TargetMesh=@n5, PolygroupID=100, TriangleIDsOut=@n6)
   > n8 = ConvertIndexListToMeshSelection(TargetMesh=@n7, IndexList=@n6, SelectionType=Vertices)
   > n9 = ConvertMeshSelectionToIndexArray(TargetMesh=@n8.ReturnValue, Selection=@n8) -> (SelectionType=Vertices)
   > n10 = GetBoneInfo(TargetMesh=@n9.ReturnValue, BoneName=foot_l)
   > n11 = For Each Loop(Array:Array of Integers=@n9) -> (Array Element:int)
   n12 = Break Bone Info(GeometryScriptBoneInfo=@n10.BoneInfo)
   > n13 = SetVertexPosition(TargetMesh=@n10.ReturnValue, VertexID=@n11, NewPosition=BreakTransform(InTransform=@n12.WorldTransform), bDeferChangeNotifications=false)
   > n14 = SetVertexBoneWeights(
    TargetMesh=@n13.ReturnValue
    VertexID=@n11
    BoneWeights=Make Array("[0]":Bone Weights Structure=Make Bone Weights(BoneIndex=@n12, Weight=1)) -> (Array:Array of Bone Weights Structures)
    Profile=(ProfileName="Default")
   )
   CopyMorphTargetToSkeletalMesh(execute=@n11.Completed, FromMorphTarget=@n14.ReturnValue, ToSkeletalMeshAsset=@n4, MorphTargetName=shoes_mid_)
  ]}
 Cobblers_TestRebind {nodes [
   Cobblers_TestRebind() {is=title K2Node_FunctionEntry}
   > n1 = Sequence!(execute) -> (then_0, then_1, then_2:Exec)
   > n2 = GetSelectionSet()
   > n3 = Cast To SkeletalMeshActor(Object="Get (a copy)"(Array=@n2))
   > n4 = Construct Object from Class(Class=&DynamicMesh)
   > n5 = CopyMeshFromSkeletalMesh(FromSkeletalMeshAsset=Get SkeletalMeshAsset(self=Get SkeletalMeshComponent(self=@n3)), ToDynamicMesh=@n4)
   n6 = GetSelectionSet(execute=@n1.then_1)
   > n7 = Cast To SkeletalMeshActor(Object="Get (a copy)"(Array=@n6, Dimension 1=1))
   > n8 = Construct Object from Class(Class=&DynamicMesh)
   n9 = Get SkeletalMeshComponent(self=@n7)
   > n10 = CopyMeshFromComponent(Component=@n9, ToDynamicMesh=@n8, bTransformToWorld=false)
   n11 = For Each Loop(Exec=@n10.Success, Array:Array of Names=GetMaterialSlotNames(self=@n9)) -> (Array Element:Name)
   n12 = CopyBonesFromMesh(
    execute=@n11.Completed
    SourceMesh=@n5
    TargetMesh=@n10.ReturnValue
    Options=Make Geometry Script Copy Bones from Mesh Options(ReindexWeights=true, BonesToCopyFromSource=AllBones)
   )
   > CreateNewSkeletalMeshAssetFromMesh(
    FromDynamicMesh=@n12
    InSkeleton=&F_MED_AgentSherbert_Skeleton
    AssetPathAndName=/Game/00_TEST/Cobblers/RebindShoes
    Options=Make Geometry Script Create New Skeletal Mesh Asset Options(
     bEnableRecomputeNormals=False
     bEnableRecomputeTangents=False
     Materials=Get MaterialsMap() -> (MaterialsMap:Map of Names to Material Interface Object References)
     bUseMeshBoneProportions=False
     bApplyNaniteSettings=False
     NaniteSettings=(bEnabled=False,bExplicitTangents=False,bLerpUVs=True,bSeparable=False,bVoxelNDF=True,bVoxelOpacity=False,ShapePreservation=None,PositionPrecision=-2147483648,NormalPrecision=-1,TangentPrecision=-1,BoneWeightPrecision=-1,TargetMinimumResidencyInKB=0,KeepPercentTriangles=1,TrimRelativeError=0,GenerateFallback=PlatformDefault,FallbackTarget=Auto,FallbackPercentTriangles=1,FallbackRelativeError=1,MaxEdgeLengthFactor=0,NumRays=64,VoxelLevel=0,RayBackUp=0,DisplacementUVChannel=0,DisplacementMaps=,NaniteAssemblyData=(Parts=,Nodes=))
     bUseOriginalVertexOrder=False
    )
   )
   Add(execute=@n11.LoopBody, TargetMap=Get MaterialsMap(), Key=@n11, Value=GetMaterialByName(self=@n9, MaterialSlotName=@n11))
  ]}
 EventGraph {links [], nodes []}
 GetPointClosestToSegmentAlpha {nodes [
   n1 = Inputs!() -> (A:Vector=~, B:Vector=~, Point:Vector=~)
   n2 = "Distance (Vector)"!(V1:Vector=@n1, V2:Vector=@n1.B) -> (ReturnValue:double)
   n3 = "Distance (Vector)"(V1=FindClosestPointOnSegment!(Point:Vector=@n1.Point, SegmentStart:Vector=@n1, SegmentEnd:Vector=@n1.B) -> (ReturnValue:Vector), V2=@n1)
   Outputs!(Alpha:double="Clamp (Float)"(Value=float / float(A=@n3, B=@n2), Min=0, Max=1))
  ]}
 PointInCapsule {nodes [
   n1 = Inputs!() -> (A=~, B=~, Radius:double=~, Point=~)
   n2 = FindClosestPointOnSegment(Point=@n1.Point, SegmentStart=@n1, SegmentEnd=@n1.B)
   Outputs!(IsInside:bool="float <= float"(A="Distance (Vector)"(V1=@n2, V2=@n1.Point), B=@n1.Radius) -> (ReturnValue=~))
  ]}
 Test Render Data Conversion {nodes [
   Test Render Data Conversion() {is=title K2Node_FunctionEntry}
   > n1 = GetSelectionSet()
   > n2 = Cast To SkeletalMeshActor(Object="Get (a copy)"(Array=@n1))
   > n3 = Construct Object from Class(Class=&DynamicMesh)
   > n4 = CopyMeshFromSkeletalMesh(
    FromSkeletalMeshAsset=Get SkeletalMeshAsset(self=Get SkeletalMeshComponent(self=@n2))
    ToDynamicMesh=@n3
    RequestedLOD=Make Geometry Script Mesh Read LOD!(LODType:EGeometryScriptLODType Enum=RenderData, LODIndex:int=1) -> (GeometryScriptMeshReadLOD:Geometry Script Mesh Read LOD Structure) {is=title K2Node_MakeStruct}
   )
   CreateNewSkeletalMeshAssetFromMesh(
    execute=@n4.Success
    FromDynamicMesh=@n4
    InSkeleton=/Game/Characters/Player/Male/Medium/Bodies/M_MED_Pirate_Progressive/Meshes/M_MED_Pirate_T3_Body_Skeleton
    AssetPathAndName=/Game/00_TEST/Cobblers/RenderData
   )
   Get Actor Transform(self=@n2)
  ]}
]
variables [
 AltSkinWeightName {type=string, default=Shoes}
 APlane {type=Plane, default=<0 0 0 0>}
 Body_Materials {type="map<MaterialInterface>", default {}}
 Body_Skeleton {type=Skeleton, default=null}
 Body_SKM_Asset {type=SkeletalMesh, default=null}
 Body_SKMesh_Component {type=SkeletalMeshComponent}
 Bone Weights {type="array<GeometryScriptBoneWeight>", default []}
 Bounds_Array {type="array<Object>", default []}
 Bounds_List {type="array<string>", default [Bounds_Left,Bounds_Right]}
 Box_Array {type="array<Box>", default []}
 Box_Left {type=Box, default {isValid=false, max=<0 0 0>, min=<0 0 0>}, shapeType=~}
 Box_Right {type=Box, default {isValid=false, max=<0 0 0>, min=<0 0 0>}, shapeType=~}
 calf_index {type=integer}
 calf_twist_l_index {type=integer}
 calf_twist_r_index {type=integer}
 CalfName_List {type="array<string>", default [calf_l,calf_r]}
 CalfName_List_Single {type="array<string>", default [calf_l]}
 CalfTwistName_List {type="array<string>", default [calf_twist_01_l, calf_twist_01_r]}
 Collapse_Location {type=Vector}
 Foot_Points {type="array<integer>"}
 FootName_List {type="array<string>", default [foot_l,foot_r]}
 Left_Bounding_Box {type=Box, default {isValid=false, max=<0 0 0>, min=<0 0 0>}, shapeType=~}
 Leg_Radius {type=number, default=12}
 Legs_SKM_Asset {type=SkeletalMesh, default=null}
 Legs_Template_Mesh {type=DynamicMesh, default=null}
 Master_SKM_Asset {type=SkeletalMesh, default=null}
 Matrix {type=Matrix, default {wPlane=<0 0 0 0>, xPlane=<0 0 0 0>, yPlane=<0 0 0 0>, zPlane=<0 0 0 0>}}
 Mesh_Vertex_Selection {type=GeometryScriptMeshSelection, default {}}
 Min {type=Vector}
 Modified_Body_Mesh {type=DynamicMesh, default=null}
 Modified_Body_Mesh_0 {type=DynamicMesh, default=null}
 Modified_Vertex_Array {type="array<integer>"}
 New Param {type="array<integer>"}
 Points {type="array<Vector>"}
 Right_Bounding_Box {type=Box, default {isValid=false, max=<0 0 0>, min=<0 0 0>}, shapeType=~}
 Shoe_Materials {type="map<MaterialInterface>", default {}}
 Shoes_SKM_Asset {type=SkeletalMesh, default=null}
 Shoes_SKMesh_Component {type=SkeletalMeshComponent}
 Socks_SKM_Asset {type=SkeletalMesh, default=/BRCosmetics/Characters/Player/Test/Shoes/Normalized/F_MED_Socks_High}
 Source Mesh {type=DynamicMesh, default=null}
 Target Profile {type=GeometryScriptBoneWeightProfile, default {profileName=Default}}
 Template_Leg_Mesh {type=DynamicMesh, default=null}
 Template_List {type="array<string>", default [Template_Mesh]}
 Template_SKMesh_Component {type=SkeletalMeshComponent}
 ThighName_List {type="array<string>", default [thigh_l,thigh_r]}
 Transfer_Body_Mesh {type=DynamicMesh, default=null}
 UnModified_Body_Mesh {type=DynamicMesh, default=null}
 vertex {type=integer}
 World Context Object_0 {type=Object, default=null}
]
