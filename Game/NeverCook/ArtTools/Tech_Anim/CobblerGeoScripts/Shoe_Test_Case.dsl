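; Test actor for the shoe Control Rig, used to pose and scale calf and foot controls on shoe and sock meshes.
; On Control Rig pre-initialize, it assigns the FMED shoes rig class and maps both skeletal meshes to its output.
; Four Move functions set local-space position and scale on the left and right calf and foot controls; nothing here calls them.
; The overlap, pre-construction and construction script entries are empty.
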(event Collision|EventActorBeginOverlap
  (params (OtherActor :type Actor%20Object%20Reference)))

(event OnPreInitialize(ControlRig)
  (params (Component :type Control%20Rig%20Component%20Object%20Reference))
  (ControlRig|SetControlRigClass Component "/BRCosmetics/Characters/Player/Test/Shoes/Normalized/FMED_Shoes_CtrlRig_01.FMED_Shoes_CtrlRig_01_C")
  (ControlRig|AddMappedSkeletalMesh Component (Variables|Default|GetSkeletalMesh_Shoes) "Output")
  (ControlRig|AddMappedSkeletalMesh Component (Variables|Default|GetSkeletalMesh_Socks) "Output"))

(event OnPreConstruction(ControlRig)
  (params (Component :type Control%20Rig%20Component%20Object%20Reference)))

(fn Move_Calf_L_Control
  (bind _controlrig (Variables|Default|GetControlRig))
  (ControlRig|SetControlPosition _controlrig "calf_l_ctrl" "0, 0, 0" "LocalSpace")
  (ControlRig|SetControlScale _controlrig "calf_l_ctrl" "1.600000,1.600000,1.600000" "LocalSpace"))

(fn Move_Calf_R_Control
  (bind _controlrig (Variables|Default|GetControlRig))
  (ControlRig|SetControlPosition _controlrig "calf_r_ctrl" "-3.000000,0.000000,0.000000" "LocalSpace")
  (ControlRig|SetControlScale _controlrig "calf_r_ctrl" "1.600000,1.600000,1.600000" "LocalSpace"))

(fn Move_Foot_L_Control
  (bind _controlrig (Variables|Default|GetControlRig))
  (ControlRig|SetControlPosition _controlrig "foot_l_ctrl" "10.000000,5.000000,0.000000" "LocalSpace")
  (ControlRig|SetControlScale _controlrig "foot_l_ctrl" "1.300000,1.300000,1.300000" "LocalSpace"))

(fn Move_Foot_R_Control
  (bind _controlrig (Variables|Default|GetControlRig))
  (ControlRig|SetControlPosition _controlrig "foot_r_ctrl" "-10.000000,-5.000000,0.000000" "LocalSpace")
  (ControlRig|SetControlScale _controlrig "foot_r_ctrl" "1.300000,1.300000,1.300000" "LocalSpace"))

(fn ConstructionScript)
